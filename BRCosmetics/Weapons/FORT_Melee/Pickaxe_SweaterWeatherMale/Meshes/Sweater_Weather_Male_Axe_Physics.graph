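# Physics asset for the Sweater Weather Male two-handed pickaxe cosmetic, letting its hat, carrot, pipe, bowtie and branch pieces move relative to the handle.
# A single long box body covers the handle, and each dyn_ bone gets a sphere body offset from its bone origin.
# Each piece is constrained directly to the handle in a star layout, sharing identical moderate swing limits and tight twist limits.
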PhysicsAsset /BRCosmetics/Weapons/FORT_Melee/Pickaxe_SweaterWeatherMale/Meshes/Sweater_Weather_Male_Axe_Physics
bodies {
 Melee_TwoHanded_Handle [{shapeType=Box, extentX=28.5278, extentY=7.5, extentZ=188.2657, center=<0 0 21.6051>}]
 dyn_bowtie [{shapeType=Sphere, radius=8.5265, center=<8.2511 0.2047 -0.1946>}]
 dyn_branch [{shapeType=Sphere, radius=7.6161, center=<9.4199 1.5386 -0.0698>}]
 dyn_carrot [{shapeType=Sphere, radius=13.2805, center=<11.6998 0.4278 1.0875>}]
 dyn_hat [{shapeType=Sphere, radius=9.6086, center=<8.5164 -1.7976 0.8174>}]
 dyn_pipe [{shapeType=Sphere, radius=12.4876, center=<9.7299 0.0471 -0.1857>}]
}
constraints [
 {bone1Name=dyn_hat, bone2Name=Melee_TwoHanded_Handle, swing1LimitDegrees=25, swing2LimitDegrees=25, twistLimitDegrees=5}
 {bone1Name=dyn_carrot, bone2Name=Melee_TwoHanded_Handle, swing1LimitDegrees=25, swing2LimitDegrees=25, twistLimitDegrees=5}
 {bone1Name=dyn_pipe, bone2Name=Melee_TwoHanded_Handle, swing1LimitDegrees=25, swing2LimitDegrees=25, twistLimitDegrees=5}
 {bone1Name=dyn_bowtie, bone2Name=Melee_TwoHanded_Handle, swing1LimitDegrees=25, swing2LimitDegrees=25, twistLimitDegrees=5}
 {bone1Name=dyn_branch, bone2Name=Melee_TwoHanded_Handle, swing1LimitDegrees=25, swing2LimitDegrees=25, twistLimitDegrees=5}
]
